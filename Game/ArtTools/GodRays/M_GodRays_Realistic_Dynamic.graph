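# Additive, unlit, two-sided material for dynamic god rays that follow the sunlight direction and take the atmospheric light color.
# Vertex offset math stretches UVs along the sun vector into tapered, camera-facing strips, using a cross product with the view vector.
# Time-of-day opacity scales UV-based width and length falloffs softened by DepthFade.
# When that opacity is not positive, the offset throws vertices far below the world, hiding the rays.
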Material /Game/ArtTools/GodRays/M_GodRays_Realistic_Dynamic
graph {
 properties {blendMode=BLEND_Additive, pixelDepthOffsetMode=PDOM_Legacy, shadingModel=MSM_Unlit, twoSided=true, usedShadingModels=Unlit}
 nodes [
  CollectionParameter() {collection=&FortniteMaterialParameters, parameterName=SunLightColor}
  CollectionParameter() {collection=&FortniteMaterialParameters, parameterName=SunLightColor}
  n1 = ScalarParameter() {defaultValue=524.1792, parameterName=Width}
  n2 = CollectionParameter() {
   collection=&FortniteMaterialParameters
   parameterName=SunlightDirectionalVector
  }
  n3 = ComponentMask(None=@n2) {b=true, g=true, r=true}
  n4 = DotProduct(A=@n3, B=Constant3Vector() {constant=<rgba 0 0 1 0>})
  Divide(A=@n1, B=@n4)
  n5 = Add(A=@n1, B=ScalarParameter() {defaultValue=1818.7042, parameterName=Width_Add_Tip})
  Divide(A=@n5, B=@n4)
  n6 = ComponentMask(None=TextureCoordinate()) {g=true}
  n7 = Multiply(A=ScalarParameter() {defaultValue=26440.8418, parameterName=Length}) {constB=-1}
  n8 = Multiply(A=Multiply(A=@n3, B=@n6), B=@n7)
  n9 = ComponentMask(None=TextureCoordinate()) {r=true}
  n10 = Multiply(A=LinearInterpolate(A=@n1, B=@n5, Alpha=@n6), B=Multiply(A=Subtract(A=@n9) {constB=0.5}) {constB=2})
  n11 = Add(A=@n8, B=WorldPosition().XYZ)
  n12 = CameraPositionWS()
  n13 = Normalize(VectorInput=CrossProduct(A=@n3, B=Subtract(A=@n11, B=@n12)))
  n14 = Constant3Vector() {constant=<rgba 0 0 -9000000159744 0>}
  n15 = Multiply(
   A=MaterialFunctionCall(
    Value_Morning=ScalarParameter() {parameterName=Opacity_Morning}
    Value_Day=ScalarParameter() {defaultValue=1, parameterName=Opacity_Day}
    Value_Evening=ScalarParameter() {parameterName=Opacity_Evening}
    Value_Night=ScalarParameter() {parameterName=Opacity_Night}
   ) {materialFunction=&Scalar_Time_of_Day}.Result
   B=ScalarParameter() {defaultValue=0.9791, parameterName=Opacity}
  )
  If(A=@n15, "A > B"=Add(A=@n8, B=Multiply(A=@n10, B=@n13)), "A == B"=@n14, "A < B"=@n14)
  LinearInterpolate(A=ScalarParameter() {defaultValue=1518.2609, parameterName=LengthExtendTop}, B=@n7, Alpha=@n6)
  Multiply(A=ScalarParameter() {defaultValue=0.16, parameterName=Brightness}, B=AtmosphericLightColor())
  n16 = TextureCoordinate()
  n17 = ComponentMask(None=@n16) {r=true}
  n18 = Multiply(A=Multiply(A=OneMinus(None=@n17), B=@n17)) {constB=4}
  n19 = ScalarParameter() {defaultValue=2.444, parameterName=Falloff_Width}
  n20 = ComponentMask(None=@n16) {g=true}
  n21 = OneMinus(None=@n20)
  n22 = ScalarParameter() {defaultValue=2, parameterName=Falloff_Length_End}
  n23 = Multiply(A=Saturate(None=Power(Base=@n18, Exp=@n19)), B=Saturate(None=Power(Base=@n21, Exp=@n22)))
  n24 = ScalarParameter() {defaultValue=3.6049, parameterName=Falloff_Bias_start}
  n25 = Saturate(None=Power(Base=LinearInterpolate(B=@n24, Alpha=@n20), Exp=ScalarParameter() {defaultValue=2, parameterName=Falloff_Length_End}))
  Multiply(A=DepthFade(Opacity=Multiply(A=@n23, B=@n25), FadeDistance=ScalarParameter() {defaultValue=3554.7412, parameterName=DepthFade}), B=@n15)
 ]
}
