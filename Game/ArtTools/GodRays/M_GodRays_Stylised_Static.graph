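# Unlit additive two-sided material for stylised god rays along a fixed Direction parameter, with time-of-day color and opacity.
# Vertex math builds camera-facing shafts from UVs, spreading and widening along their length, and collapses them to the object origin when fully faded.
# A panning noise mask with edge falloffs and depth fade is attenuated by view angle, camera height, horizontal distance, and switchable safe-zone and Hades fades.
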Material /Game/ArtTools/GodRays/M_GodRays_Stylised_Static
graph {
 properties {
  bUseTranslucencyVertexFog=false
  blendMode=BLEND_Additive
  pixelDepthOffsetMode=PDOM_Legacy
  shadingModel=MSM_Unlit
  twoSided=true
  usedShadingModels=Unlit
 }
 nodes [
  AtmosphericLightColor()
  CollectionParameter() {collection=&FortniteMaterialParameters, parameterName=SunLightColor}
  CollectionParameter() {
   collection=&FortniteMaterialParameters
   parameterName=SunlightDirectionalVector
  }
  CollectionParameter() {collection=&FortniteMaterialParameters, parameterName=SunLightColor}
  Constant3Vector() {constant=<rgba 0 0 -9000000159744 0>}
  n1 = ScalarParameter() {defaultValue=2386.4924, parameterName=Width}
  n2 = Normalize(VectorInput=VectorParameter() {parameterName=Direction, defaultValue=<rgba -0.128 0 -1 1>}.RGB)
  n3 = ComponentMask(None=@n2) {b=true, g=true, r=true}
  n4 = DotProduct(A=@n3, B=Constant3Vector() {constant=<rgba 0 0 1 0>})
  Divide(A=@n1, B=@n4)
  n5 = Add(A=@n1, B=ScalarParameter() {defaultValue=8793.9521, parameterName=Width_Add_Tip})
  Divide(A=@n5, B=@n4)
  n6 = TextureCoordinate()
  n7 = WorldPosition() {worldPositionShaderOffset=WPT_ExcludeAllShaderOffsets}
  n8 = ObjectPositionWS()
  n9 = MaterialFunctionCall(Vector=Multiply(A=Subtract(A=@n7.XYZ, B=@n8), B=Constant3Vector() {constant=<rgba 1 1 0 0>}), Default=Constant3Vector()) {materialFunction=&SafeNormalize}
  n10 = ScalarParameter() {defaultValue=8634.5088, parameterName=Spread}
  n11 = ScalarParameter() {defaultValue=1518.2609, parameterName=LengthExtendTop}
  n12 = ScalarParameter() {defaultValue=254383.3438, parameterName=Length}
  n13 = ComponentMask(None=TextureCoordinate()) {g=true}
  n14 = Add(
   A=Multiply(A=ComponentMask(None=@n6) {g=true}, B=Multiply(A=@n9.Result, B=@n10))
   B=Multiply(A=@n3, B=LinearInterpolate(A=@n11, B=@n12, Alpha=@n13))
  )
  n15 = ComponentMask(None=TextureCoordinate()) {r=true}
  n16 = Multiply(A=LinearInterpolate(A=@n1, B=@n5, Alpha=@n13), B=Multiply(A=Subtract(A=@n15) {constB=0.5}) {constB=2})
  n17 = Add(A=@n14, B=WorldPosition().XYZ)
  n18 = CameraPositionWS()
  n19 = Normalize(VectorInput=CrossProduct(A=@n3, B=Subtract(A=@n17, B=@n18)))
  n20 = Subtract(A=ObjectPositionWS(), B=WorldPosition().XYZ)
  n21 = ObjectPositionWS()
  n22 = Constant3Vector() {constant=<rgba 1 1 0 0>}
  n23 = CameraPositionWS()
  n24 = Subtract(
   A=Distance(A=Multiply(A=@n21, B=@n22), B=Multiply(A=@n22, B=@n23))
   B=ScalarParameter() {defaultValue=28593.5566, parameterName=DistanceFade_Offset}
  )
  n25 = ScalarParameter() {defaultValue=8995.6699, parameterName=DistanceFade_Length}
  n26 = ScalarParameter() {defaultValue=1, parameterName=Opacity_Morning}
  n27 = ScalarParameter() {defaultValue=1, parameterName=Opacity_Day}
  n28 = ScalarParameter() {defaultValue=1, parameterName=Opacity_Evening}
  n29 = ScalarParameter() {defaultValue=0.15, parameterName=Opacity_Night}
  n30 = Multiply(
   A=Saturate(None=Divide(A=@n24, B=@n25))
   B=Multiply(
    A=MaterialFunctionCall(Value_Morning=@n26, Value_Day=@n27, Value_Evening=@n28, Value_Night=@n29) {materialFunction=&Scalar_Time_of_Day}.Result
    B=ScalarParameter() {defaultValue=0.476, parameterName=Opacity}
   )
  )
  n31 = ScalarParameter() {defaultValue=0.112, parameterName=HadesDarken}
  n32 = CollectionParameter() {collection=&FortniteMaterialParameters, parameterName=HadesActive}
  n33 = StaticSwitchParameter(True=Multiply(A=@n30, B=LinearInterpolate(B=@n31, Alpha=@n32) {constA=1}), False=@n30) {parameterName=UseHadesFade}
  n34 = CollectionParameter() {collection=&FortniteMaterialParameters, parameterName=AthenaSafeZoneCenter}
  n35 = Distance(A=TransformPosition(None=Constant3Vector()) {transformType=TRANSFORMPOSSOURCE_World}, B=ComponentMask(None=@n34) {b=true, g=true, r=true})
  n36 = CollectionParameter() {collection=&FortniteMaterialParameters, parameterName=AthenaSafeZoneRadius}
  n37 = ScalarParameter() {defaultValue=29370.9648, parameterName=SafeZoneBlendDistance}
  n38 = Subtract(A=@n36, B=@n37)
  n39 = Saturate(None=Divide(A=Subtract(A=@n35, B=@n38), B=@n37))
  n40 = StaticSwitchParameter(True=Multiply(A=@n33, B=OneMinus(None=@n39)), False=@n33) {parameterName=UseSafeZoneFade}
  If(A=@n40, "A > B"=Add(A=@n14, B=Multiply(A=@n16, B=@n19)), "A == B"=@n20, "A < B"=@n20)
  n41 = VectorParameter() {parameterName=Color, defaultValue=<rgba 1 0.4788 0.099 1>}
  Multiply(
   A=ScalarParameter() {defaultValue=1, parameterName=Brightness}
   B=MaterialFunctionCall(
    Color_Morning=@n41.RGB
    Color_Day=@n41.RGB
    Color_Evening=VectorParameter() {parameterName=ColorEvening, defaultValue=<rgba 1 0.4788 0.099 1>}.RGB
    Color_Night=VectorParameter() {parameterName=Colornight, defaultValue=<rgba 0.3385 0.4177 1 1>}.RGB
   ) {materialFunction=&Color_Time_of_Day}.Result
  )
  Multiply(A=@n3, B=@n13)
  Multiply()
  n42 = CameraPositionWS()
  n43 = Divide(
   A=Subtract(A=ComponentMask(None=@n42) {b=true}, B=ScalarParameter() {defaultValue=50000, parameterName=HeightFade_Offset})
   B=ScalarParameter() {defaultValue=69161.1875, parameterName=HeightFade}
  )
  n44 = DotProduct(A=CameraVectorWS(), B=@n2)
  n45 = Panner(
   Coordinate=Multiply(A=TextureCoordinate() {coordinateIndex=1}, B=ScalarParameter() {defaultValue=0.352, parameterName=UV_Scale})
   Speed=ScalarParameter() {defaultValue=-0.02, parameterName=NoisePanSpeed}
  )
  n46 = TextureCoordinate()
  n47 = ComponentMask(None=@n46) {r=true}
  n48 = Multiply(A=Multiply(A=OneMinus(None=@n47), B=@n47)) {constB=4}
  n49 = ScalarParameter() {defaultValue=1.5026, parameterName=Falloff_Width}
  n50 = ComponentMask(None=@n46) {g=true}
  n51 = OneMinus(None=@n50)
  n52 = ScalarParameter() {defaultValue=1.2127, parameterName=Falloff_Length_End}
  n53 = Multiply(A=Saturate(None=Power(Base=@n48, Exp=@n49)), B=Saturate(None=Power(Base=@n51, Exp=@n52)))
  n54 = ScalarParameter() {defaultValue=2.0715, parameterName=Falloff_Bias_start}
  n55 = Saturate(None=Power(Base=LinearInterpolate(B=@n54, Alpha=@n50), Exp=ScalarParameter() {defaultValue=3.7583, parameterName=Falloff_Length_Start}))
  n56 = DepthFade(
   Opacity=Multiply(A=TextureSample(UVs=@n45) {samplerType=SAMPLERTYPE_Masks, texture=&/GodRays_M}.R, B=Multiply(A=@n53, B=@n55))
   FadeDistance=ScalarParameter() {defaultValue=3554.7412, parameterName=DepthFade}
  )
  Multiply(A=OneMinus(None=Saturate(None=@n43)), B=Multiply(A=OneMinus(None=@n44), B=Multiply(A=@n56, B=@n40)))
  Multiply(A=@n12) {constB=-1}
  OneMinus(None=@n32)
  ScalarParameter() {defaultValue=190749.8906, parameterName=SafeZoneBlendDistance+Debug}
 ]
}
